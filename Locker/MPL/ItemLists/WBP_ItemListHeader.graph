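# WBP_ItemListHeader is a UserWidget header for Locker item lists, showing a title, an item count and a divider line in one horizontal row.
# UpdateHeaderTitle sets the title text, and UpdateHeaderCount formats an integer in parentheses into the dimmed count text.
# A half-opacity gradient image fills the remaining width as the divider.
# The event graph holds only empty event stubs, so content changes only through those two functions.
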WidgetBlueprint /Locker/MPL/ItemLists/WBP_ItemListHeader
graphs [
 EventGraph {nodes [
   Event PreConstruct()
   Event Construct()
   Event Tick()
  ]}
 UpdateHeaderCount {nodes [
   n1 = UpdateHeaderCount!() -> (then, ItemCount:int) {is=title K2Node_FunctionEntry}
   > "SetText (Text)"(
    self=Get ItemCountText() -> (ItemCountText:Common Text Object Reference)
    InText=Format Text!(Format="LOCTEXT(\"({ItemCount})\")", ItemCount:Text="To Text (Integer)"(Value=@n1)) -> (Result)
   )
  ]}
 UpdateHeaderTitle {nodes [
   n1 = UpdateHeaderTitle!() -> (then, Title:Text) {is=title K2Node_FunctionEntry}
   > "SetText (Text)"(self=Get TitleText() -> (TitleText:Common Text Object Reference), InText=@n1)
  ]}
]
variables []
widgets {
 info {inheritedWidgetCount=0, parentClass=&UserWidget, rootWidgetClass=&HorizontalBox, widgetCount=4}
 widgets [
  HorizontalBox_27 {widgetClassPath=&HorizontalBox, widget=HorizontalBox_27}
  TitleText {
   widgetClassPath=&CommonTextBlock
   bIsVariable=true
   parent=HorizontalBox_27
   widget=TitleText
   properties {
    bIsScrollingEnabled=false
    lineHeightPercentage=0.9
    scrollOrientation=Orient_Vertical
    scrollStyle=/FortUILibrary/Styles/Typography/Foundation/TextScrollStyle-NoFade.TextScrollStyle-NoFade_C
    text=Title
    wrappingPolicy=AllowPerCharacterWrapping
    font {
     bForceMonospaced=false
     bMaterialIsStencil=false
     fontMaterial=null
     fontObject=&HeadingNow
     letterSpacing=10
     monospacedWidth=1
     size=15.75
     skewAmount=0
     typefaceFontName=74 Regular
     outlineSettings {
      bApplyOutlineToDropShadows=false
      bMiteredCorners=false
      bSeparateFillAlpha=false
      outlineMaterial=null
      outlineSize=0
      outlineColor=<rgba 0 0 0 1>
     }
    }
    shadowOffset=<0 0>
   }
   slot {class=HorizontalBoxSlot, properties {horizontalAlignment=HAlign_Left, verticalAlignment=VAlign_Center, padding=<ltrb 10 0 8 0>}}
  }
  ItemCountText {
   widgetClassPath=&CommonTextBlock
   bIsVariable=true
   parent=HorizontalBox_27
   widget=ItemCountText
   properties {
    bIsScrollingEnabled=false
    lineHeightPercentage=0.9
    scrollOrientation=Orient_Vertical
    scrollStyle=&TextScrollStyle-NoFade_C
    text=(ItemCount)
    textOverflowPolicy=Ellipsis
    wrappingPolicy=AllowPerCharacterWrapping
    colorAndOpacity {colorUseRule=UseColor_Specified, specifiedColor=<rgba 1 1 1 0.6>}
    font {
     bForceMonospaced=false
     bMaterialIsStencil=false
     fontMaterial=null
     fontObject=&HeadingNow
     letterSpacing=10
     monospacedWidth=1
     size=15.75
     skewAmount=0
     typefaceFontName=74 Regular
     outlineSettings {
      bApplyOutlineToDropShadows=false
      bMiteredCorners=false
      bSeparateFillAlpha=false
      outlineMaterial=null
      outlineSize=0
      outlineColor=<rgba 0 0 0 1>
     }
    }
    shadowOffset=<0 0>
   }
   slot {class=HorizontalBoxSlot, properties {horizontalAlignment=HAlign_Left, verticalAlignment=VAlign_Center, padding=<ltrb 0 0 5.08 0>}}
  }
  divider {
   widgetClassPath=&Image
   bIsVariable=true
   parent=HorizontalBox_27
   widget=divider
   properties {
    renderOpacity=0.5
    visibility=SelfHitTestInvisible
    brush {
     bIsDynamicallyLoaded=false
     drawAs=Image
     imageType=FullColor
     mirroring=NoMirror
     resourceName=None
     resourceObject=/Locker/MPL/Materials/MI_UI_LeftGradientCarousel
     tiling=NoTile
     imageSize=<50 2>
     margin=<ltrb 0 0 0 0>
     outlineSettings {
      bUseBrushTransparency=false
      roundingType=HalfHeightRadius
      width=0
      color {colorUseRule=UseColor_Specified, specifiedColor=<rgba 0 0 0 0>}
      cornerRadii=<0 0 0 0>
     }
     tintColor {colorUseRule=UseColor_Specified, specifiedColor=<rgba 1 1 1 0.5>}
     uVRegion {bIsValid=false, max=<0 0>, min=<0 0>}
    }
   }
   slot {class=HorizontalBoxSlot, properties {verticalAlignment=VAlign_Center, padding=<ltrb 0 12 0 12>, size {sizeRule=Fill, value=1}}}
  }
 ]
}
