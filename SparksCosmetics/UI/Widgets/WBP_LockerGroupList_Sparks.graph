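# Row widget for the Sparks locker list, showing one collection group row as non-scrolling tiles of Jam item cards, with a title-and-count header on each group's first row.
# Setting the row object fills the tiles, then after a short delay configures displayed cards and relays their events through base-list delegates.
# Hover and selection bind preview actions outside emote-wheel mode.
# It also handles gamepad focus navigation, compact mobile sizing and impression telemetry.
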WidgetBlueprint /SparksCosmetics/UI/Widgets/WBP_LockerGroupList_Sparks
graphs [
 AddImpressionTelemetry {nodes [
   AddImpressionTelemetry!() -> (then, TabFlowRiderString:String) {is=title K2Node_FunctionEntry}
   > n1 = For Each Loop(Array:Array of User Widget Object References=GetDisplayedEntryWidgets!(self:List View Base Object Reference=Get LockerTiles() -> (LockerTiles:Fort Tile View Object Reference)) -> (ReturnValue:Array of User Widget Object References)) -> (Array Element:User Widget Object Reference)
   n2 = Get SectionContentsArray() -> (SectionContentsArray:Array of Flow Analytics Impression Section Content Structures)
   Branch(
    execute=@n1.Completed
    Condition="integer > integer"(A=Length(TargetArray:Array of Flow Analytics Impression Section Content Structures=@n2))
   )
   > "Fire Event Add Impression (Non-Localized)"!(
    execute
    SectionID:String=Get TabFlowRiderString() -> (TabFlowRiderString:String)
    SectionContents:Array of Flow Analytics Impression Section Content Structures=Get SectionContentsArray()
   ) -> (then)
   n3 = Cast To WBP_UIKit_ItemCard_Locker_Jam!(execute=@n1.LoopBody, Object=@n1) -> (then, CastFailed, AsWBP UIKit Item Card Locker Jam:WBP UIKit Item Card Locker Jam Object Reference) {is=title K2Node_DynamicCast}
   > Is Valid(InputObject=@n3)
   n4 = Get FortItemVM!(self:WBP UIKit Item Card Locker Jam Object Reference=@n3) -> (FortItemVM:Fort Item VM Object Reference)
   > Branch(Condition=NOT Boolean(A=Get bIsShuffleTile!(self:Fort Item VM Object Reference=@n4) -> (bIsShuffleTile:bool)))
   n5 = GetItemTemplateId!(self:Fort Item VM Object Reference=@n4) -> (ReturnValue:String)
   > Branch(Condition=NOT Boolean(A=IsEmpty(InString=@n5)))
   > Add(
    TargetArray:Array of Flow Analytics Impression Section Content Structures=Get SectionContentsArray()
    NewItem:"Flow Analytics Impression Section Content Structure (by ref)"=Make Flow Analytics Impression Section Content!(ContentID:Text="To Text (String)"(InString=@n5), OfferID:Text, Row:int=-1, Column:int=-1, Index:int=-1) -> (FlowAnalyticsImpressionSectionContent:Flow Analytics Impression Section Content Structure) {is=title K2Node_MakeStruct}
   ) {is=title K2Node_CallArrayFunction}
  ]}
 BindFunctions {nodes [
   BindFunctions() {is=title K2Node_FunctionEntry}
   > n1 = Sequence!(execute) -> (then_0, then_1, then_2:Exec, then_3:Exec, then_4:Exec, then_5:Exec, then_6:Exec)
   Bind Event to On Item Double Clicked(
    execute=@n1.then_3
    self:WBP UIKit Item Card Locker Jam Object Reference=Get Cached WBP UIKit Item Card Locker Jam() -> (Cached WBP UIKit Item Card Locker Jam:WBP UIKit Item Card Locker Jam Object Reference)
    Delegate=Create Event()
   ) {is=title K2Node_AddDelegate}
   Bind Event to On Hovered(execute=@n1.then_2, self:Common Button Base Object Reference=Get Cached WBP UIKit Item Card Locker Jam(), Delegate=Create Event()) {is=title K2Node_AddDelegate}
   Bind Event to On Context Menu Closed(
    execute=@n1.then_1
    self:WBP UIKit Item Card Locker Jam Object Reference=Get Cached WBP UIKit Item Card Locker Jam()
    Delegate=Create Event()
   ) {is=title K2Node_AddDelegate}
   Bind Event to On Context Menu Opened(
    execute=@n1.then_0
    self:WBP UIKit Item Card Locker Jam Object Reference=Get Cached WBP UIKit Item Card Locker Jam()
    Delegate=Create Event()
   ) {is=title K2Node_AddDelegate}
   Bind Event to On Context Menu Edit Styles Clicked(
    execute=@n1.then_4
    self:WBP UIKit Item Card Locker Jam Object Reference=Get Cached WBP UIKit Item Card Locker Jam()
    Delegate=Create Event()
   ) {is=title K2Node_AddDelegate}
   Bind Event to Go to Locker Item Details(
    execute=@n1.then_5
    self:WBP UIKit Item Card Locker Jam Object Reference=Get Cached WBP UIKit Item Card Locker Jam()
    Delegate=Create Event()
   ) {is=title K2Node_AddDelegate}
  ]}
 ContextMenuClosedBind {nodes [
   ContextMenuClosedBind() {is=title K2Node_FunctionEntry}
   > Call On Tile Context Menu Closed!(execute:Exec, self:WBP Locker List Base Object Reference) -> (then:Exec) {is=title K2Node_CallDelegate}
  ]}
 ContextMenuOpenedBind {nodes [
   n1 = ContextMenuOpenedBind!() -> (then, ContextMenuWidgetParentEntry:Widget Object Reference) {is=title K2Node_FunctionEntry}
   > Call On Tile Context Menu Open!(execute:Exec, self:WBP Locker List Base Object Reference, Widget:Widget Object Reference=@n1) -> (then:Exec) {is=title K2Node_CallDelegate}
  ]}
 EventGraph {nodes [
   n1 = "On Item Selection Changed (LockerTiles)"!() -> (OutputDelegate, then, Item:Object Reference, bIsSelected:bool) {is=title K2Node_ComponentBoundEvent}
   n2 = Get IsEmoteWheel() -> (IsEmoteWheel:bool)
   > n3 = Branch(Condition=AND Boolean(A=@n1.bIsSelected, B=NOT Boolean(A=@n2)))
   n4 = Get CachedPreviewWidget() -> (CachedPreviewWidget:Fort Item Preview Widget Object Reference)
   PreviewActions!(execute=[@n7.then, @n3.then], self:WBP Locker List Base Object Reference, PreviewWidget:Fort Item Preview Widget Object Reference=@n4) -> (then)
   > SetupPreviewActionBinds!(execute, self:WBP Locker List Base Object Reference, PreviewWidget:Fort Item Preview Widget Object Reference=@n4) -> (then)
   n5 = "On Item Is Hovered Changed (LockerTiles)"!() -> (OutputDelegate, then, Item:Object Reference, bIsHovered:bool) {is=title K2Node_ComponentBoundEvent}
   n6 = Get IsEmoteWheel()
   > n7 = Branch(Condition=AND Boolean(A=@n5.bIsHovered, B=NOT Boolean(A=@n6)))
   n8 = BP_OnItemScrolledIntoView_Event!() -> (OutputDelegate, then, Item:Object Reference, Widget:User Widget Object Reference) {is=title K2Node_CustomEvent}
   > Unbind Event from On Item Scrolled Into View(self:List View Object Reference=Get LockerTiles(), Delegate=@n8) {is=title K2Node_RemoveDelegate}
   > Toggle Selected Context Internal()
   Event Construct()
   > Branch(Condition=IsMobileGame())
   > MobileSetup()
   n9 = Event On Item Selection Changed!() -> (OutputDelegate, then, bIsSelected:bool) {is=title K2Node_Event}
   > n10 = Branch(Condition=@n9.bIsSelected)
   Clear Selection(execute=@n10.else, self:List View Object Reference=Get LockerTiles())
   n11 = Event SetHeaderIndex!() -> (OutputDelegate, then, FirstRowGroupArray:Array of Integers) {is=title K2Node_Event}
   > SetRowHeaderInfo!(execute, self, HeaderFirstRowArray:Array of Integers=@n11.FirstRowGroupArray) -> (then)
   n12 = Event GetLastFocusedColumn!() -> (OutputDelegate, then, LastFocusedColumnIndex:int=0) {is=title K2Node_Event}
   > Set Last Focused Column Index(Last Focused Column Index:int=Select(Option 0:int=@n12.LastFocusedColumnIndex, Option 1:int=0, Index="integer < integer"(A=@n12.LastFocusedColumnIndex)) -> (ReturnValue:int)) -> (Output_Get:int=0)
   Event OnAddedToFocusPath!() -> (OutputDelegate, then, InFocusEvent:Focus Event Structure) {is=title K2Node_Event}
   n13 = Get CommonInputSubsystem()
   > Branch(Condition="Equal (Enum)"(A=GetCurrentInputType(self=@n13), B=Gamepad))
   > Branch(Condition=Get AllowNavOnFocus() -> (AllowNavOnFocus:bool))
   > NavigateToItemIndex!(execute, self, ItemIndex:int=Get Last Focused Column Index() -> (Last Focused Column Index:int=0)) -> (then)
   n14 = Event UpdateAllowNavOnFocus!() -> (OutputDelegate, then, AllowNavOnFocus:bool) {is=title K2Node_Event}
   > Set AllowNavOnFocus(AllowNavOnFocus:bool=@n14.AllowNavOnFocus) -> (Output_Get:bool)
   n15 = Event SetupTelemetry!() -> (OutputDelegate, then, TabFlowRiderString:String) {is=title K2Node_Event}
   > AddImpressionTelemetry!(execute, self:WBP Locker List Base Object Reference, TabFlowRiderString:String=@n15.TabFlowRiderString) -> (then) {is=title K2Node_CallFunction}
   n16 = Event UpdateIsEmoteWheel!() -> (OutputDelegate, then, IsEmoteWheel:bool) {is=title K2Node_Event}
   > Set IsEmoteWheel(IsEmoteWheel:bool=@n16.IsEmoteWheel) -> (Output_Get:bool)
   Event SetContextVisibility() {is=title K2Node_Event}
   > n17 = For Each Loop(Array=GetDisplayedEntryWidgets(self=Get LockerTiles()))
   > n18 = Cast To WBP_UIKit_ItemCard_Locker_Jam(Object=@n17)
   > SetVisibility(
    self=Get ContextMenu!(self:WBP UIKit Item Card Locker Jam Object Reference=@n18) -> (ContextMenu:WBP UIKit Context Menu Anchor Button Object Reference)
    InVisibility=Hidden
   )
   Event ToggleSelectedContext() {is=title K2Node_Event}
   > Toggle Selected Context Internal()
   n19 = Event BindPreviewActions!() -> (OutputDelegate, then, Preview Widget:Fort Item Preview Widget Object Reference) {is=title K2Node_Event}
   > Branch(Condition=NOT Boolean(A=Get IsEmoteWheel()))
   > PreviewActions(PreviewWidget=@n19.Preview Widget)
   n20 = Event SetupPreviewActionBinds!() -> (OutputDelegate, then, PreviewWidget:Fort Item Preview Widget Object Reference) {is=title K2Node_Event}
   > Branch(Condition=NOT Boolean(A=Get IsEmoteWheel()))
   > n21 = For Each Loop with Break!(Exec:Exec, Array:Array of User Widget Object References=GetDisplayedEntryWidgets(self=Get LockerTiles()), Break:Exec=@n23.then) -> (LoopBody:Exec, Array Element:User Widget Object Reference, Array Index:int, Completed:Exec)
   > n22 = Cast To WBP_UIKit_ItemCard_Locker_Jam(Object=@n21)
   > Branch(Condition=GetSelected!(self:Common Button Base Object Reference=@n22) -> (ReturnValue:bool))
   > Set Hovered Item(Hovered Item:WBP UIKit Item Card Locker Jam Object Reference=@n22) -> (Output_Get:WBP UIKit Item Card Locker Jam Object Reference)
   > n23 = PreviewActions(PreviewWidget=@n20.PreviewWidget)
   n24 = Event SetupBinds!() -> (OutputDelegate, then, Widget:Widget Object Reference, PreviewWidget:Fort Item Preview Widget Object Reference) {is=title K2Node_Event}
   > Branch(Condition=NOT Boolean(A=Get IsEmoteWheel()))
   > Set CachedPreviewWidget(CachedPreviewWidget:Fort Item Preview Widget Object Reference=@n24.PreviewWidget) -> (Output_Get:Fort Item Preview Widget Object Reference)
   n25 = Event OnListItemObjectSet()
   > n26 = Set List Item Object(List Item Object:Object Reference=@n25.ListItemObject) -> (Output_Get:Object Reference)
   > n27 = Cast To FortCollectionGroupRow!(execute, Object=@n26) -> (then, CastFailed, AsFort Collection Group Row:Fort Collection Group Row Object Reference) {is=title K2Node_DynamicCast}
   > Set GroupName(GroupName:Text=Get Name!(self:Fort Collection Group Row Object Reference=@n27) -> (Name:Text)) -> (Output_Get:Text)
   n28 = Get LockerTiles()
   > ClearListItems(self=@n28)
   > Set List Items!(
    execute
    self:List View Object Reference=@n28
    InListItems:Array of Object References=Get RowItems!(self:Fort Collection Group Row Object Reference=@n27) -> (RowItems:Array of Fort Item VM Object References)
   ) -> (then) {is=title K2Node_CallFunction}
   > SetScrollbarVisibility!(execute, self:List View Base Object Reference=@n28, InVisibility:ESlateVisibility Enum=Collapsed) -> (then)
   > Delay(Duration=0.2)
   > n29 = For Each Loop(Array=GetDisplayedEntryWidgets(self=Get LockerTiles()))
   > n30 = Cast To WBP_UIKit_ItemCard_Locker_Jam(Object=@n29)
   > SetVisibility(
    self=Get ContextMenu_Attachment!(self:WBP UIKit Item Card Locker Jam Object Reference=@n30) -> (ContextMenu_Attachment:Fort Visual Attachment Object Reference)
    InVisibility=Select(Option 0:ESlateVisibility Enum=Visible, Option 1:ESlateVisibility Enum=Collapsed, Index=Get IsEmoteWheel()) -> (ReturnValue:ESlateVisibility Enum)
   )
   > SetVisibility(
    self=Get Context Menu!(self:WBP UIKit Item Card Base Object Reference=@n30) -> (Context Menu:WBP UIKit Context Menu Anchor Button Base Object Reference)
    InVisibility=Select(Option 0=Hidden, Option 1=Visible, Index=GetSelected(self=@n30))
   )
   > Set Cached WBP UIKit Item Card Locker Jam(Cached WBP UIKit Item Card Locker Jam:WBP UIKit Item Card Locker Jam Object Reference=@n30) -> (Output_Get:WBP UIKit Item Card Locker Jam Object Reference)
   > BindFunctions() {is=title K2Node_CallFunction}
   > Branch(Condition=NOT Boolean(A=Get IsEmoteWheel()))
   n31 = Get CachedPreviewWidget()
   > PreviewActions(PreviewWidget=@n31)
   > SetupPreviewActionBinds(PreviewWidget=@n31)
   n32 = Event SetGroupCounts!() -> (OutputDelegate, then, ItemCountPerGroup:Array of Integers) {is=title K2Node_Event}
   > Set Item Count Per Group(Item Count Per Group:Array of Integers=@n32.ItemCountPerGroup) -> (Output_Get:Array of Integers)
   n33 = Get Selected Item!(self:List View Object Reference=Get LockerTiles()) -> (ReturnValue:Object Reference) {is=title K2Node_CallFunction}
   Is Valid(InputObject=@n33)
   > Unbind Event from On Item Scrolled Into View(self=Get LockerTiles(), Delegate=@n8)
   > Assign On Item Scrolled Into View!(execute:Exec, self:List View Object Reference=Get LockerTiles(), Delegate:"Delegate (by ref)"=@n8) -> (then:Exec) {is=title K2Node_AssignDelegate}
   > Scroll Item Into View!(execute, self:List View Object Reference=Get LockerTiles(), Item:Object Reference=@n33) -> (then)
   Set IsEmoteWheel!(execute, IsEmoteWheel=Get IsEmoteWheel(), self:WBP UIKit Item Card Locker Object Reference) -> (then, Output_Get)
  ]}
 GetEntryObject {nodes [
   GetEntryObject() {is=title K2Node_FunctionEntry}
   > Return Node!(execute, Object:Object Reference=Get List Item Object() -> (List Item Object:Object Reference))
  ]}
 GetListView {nodes [
   GetListView() {is=title K2Node_FunctionEntry}
   > Return Node!(execute, TileView:Fort Tile View Object Reference=Get LockerTiles())
  ]}
 GoToLockerDetailsBind1 {nodes [
   GoToLockerDetailsBind1!() -> (then, CustomItemDetails:User Widget Class Reference) {is=title K2Node_FunctionEntry}
   > Call Go to Locker Item Details!(execute:Exec, self:WBP Locker List Base Object Reference, CustomItemDetails:User Widget Class Reference) -> (then:Exec) {is=title K2Node_CallDelegate}
  ]}
 MobileSetup {nodes [
   MobileSetup() {is=title K2Node_FunctionEntry}
   > Branch(Condition=IsUsingCompactUI())
   > SetEntryHeight!(execute, self:Tile View Object Reference=Get LockerTiles(), NewHeight:float=140) -> (then)
  ]}
 NavigateToItemIndex {nodes [
   NavigateToItemIndex!() -> (then, ItemIndex:int) {is=title K2Node_FunctionEntry}
   n1 = int - int(A=GetNumItems!(self:List View Object Reference=Get LockerTiles()) -> (ReturnValue:int=0))
   n2 = Select(Option 0:int=Get ItemIndex() -> (ItemIndex:int=0), Option 1:int=@n1, Index="integer < integer"(A=@n1, B=Get ItemIndex() -> (ItemIndex=0))) -> (ReturnValue:int)
   n3 = Get LockerTiles()
   > NavigateToIndex!(execute, self:List View Object Reference=@n3, Index:int=@n2) -> (then)
   > SetSelectedIndex!(execute, self:List View Object Reference=@n3, Index:int=@n2) -> (then)
  ]}
 OnContextMenuEditStylesClickedBind {nodes [
   OnContextMenuEditStylesClickedBind() {is=title K2Node_FunctionEntry}
   > Call Context Edit Styles Clicked!(execute:Exec, self:WBP Locker List Base Object Reference) -> (then:Exec) {is=title K2Node_CallDelegate}
  ]}
 OnItemCardHoveredBind {nodes [
   n1 = OnItemCardHoveredBind!() -> (then, Button:Common Button Base Object Reference) {is=title K2Node_FunctionEntry}
   > n2 = Cast To WBP_UIKit_ItemCard_Locker_Jam(Object=@n1)
   > Set Hovered Item(Hovered Item=@n2)
   n3 = Get IsEmoteWheel()
   n4 = GetCurrentInputType(self=Get CommonInputSubsystem())
   > Branch(Condition=AND Boolean(A=NOT Boolean(A=@n3), B="Not Equal (Enum)"!(A:ECommonInputType Enum=@n4, B:ECommonInputType Enum=Gamepad) -> (ReturnValue:bool)))
   > PreviewActions(PreviewWidget=Get CachedPreviewWidget())
  ]}
 OnItemDoubleClickedBind {nodes [
   OnItemDoubleClickedBind() {is=title K2Node_FunctionEntry}
   > Call Close Item Selection!(execute:Exec, self:WBP Locker List Base Object Reference) -> (then:Exec) {is=title K2Node_CallDelegate}
  ]}
 PreviewActions {nodes [
   n1 = PreviewActions!() -> (then, PreviewWidget:Fort Item Preview Widget Object Reference) {is=title K2Node_FunctionEntry}
   n2 = Get Hovered Item() -> (Hovered Item:WBP UIKit Item Card Locker Jam Object Reference)
   > Is Valid(InputObject=@n2)
   > Bind Preview Actions!(
    execute
    self:WBP UIKit Item Card Base Object Reference=@n2
    Preview Widget:Fort Item Preview Widget Object Reference=@n1
    Context Menu:WBP UIKit Context Menu Anchor Button Base Object Reference=Get ContextMenu(self=@n2)
    Item VM:Fort Item VM Object Reference=Get FortItemVM(self=@n2)
    Include Special Actions:bool=true
    Include Toggle Actions:bool=true
   ) -> (then)
  ]}
 SetRowHeaderInfo {nodes [
   SetRowHeaderInfo!() -> (then, HeaderFirstRowArray:Array of Integers) {is=title K2Node_FunctionEntry}
   > Set HeaderIndexFound(HeaderIndexFound:bool=false) -> (Output_Get:bool)
   > n1 = For Each Loop(Array:Array of Integers=Get HeaderFirstRowArray() -> (HeaderFirstRowArray:Array of Integers)) -> (Array Element:int)
   > Branch(Condition="Equal (Integer)"(A=@n1, B=GetListItemIndex!(UserObjectListEntry:User Object List Entry Interface) -> (ReturnValue:int=0)))
   > Set HeaderIndexFound(HeaderIndexFound=true)
   > Set GroupCount(GroupCount:int="Get (a copy)"(Array:Array of Integers=Get Item Count Per Group() -> (Item Count Per Group:Array of Integers), Dimension 1=@n1.Array Index) -> (Output:int)) -> (Output_Get:int=0)
   n2 = Get HeaderIndexFound() -> (HeaderIndexFound:bool)
   SetGridSlotPadding!(
    execute=@n1.Completed
    Widget:Widget Object Reference=Get LockerTiles()
    Padding_Left:float=0
    Padding_Top:float=Select(Option 0:float=0, Option 1:float=40, Index=@n2) -> (ReturnValue:float)
    Padding_Right:float=0
    Padding_Bottom:float=0
   ) -> (then)
   > SetVisibility(
    self=Get ItemListHeader() -> (ItemListHeader:WBP Item List Header Object Reference)
    InVisibility=Select(Option 0:ESlateVisibility Enum=Collapsed, Option 1:ESlateVisibility Enum=SelfHitTestInvisible, Index=@n2) -> (ReturnValue:ESlateVisibility Enum)
   )
   > Branch(Condition=Get HeaderIndexFound())
   n3 = Get ItemListHeader()
   > UpdateHeaderTitle!(execute, self:WBP Item List Header Object Reference=@n3, Title:Text=Get GroupName() -> (GroupName:Text)) -> (then)
   > UpdateHeaderCount!(execute, self:WBP Item List Header Object Reference=@n3, ItemCount:int=Get GroupCount() -> (GroupCount:int=0)) -> (then)
  ]}
 Toggle Selected Context Internal {nodes [
   Toggle Selected Context Internal() {is=title K2Node_FunctionEntry}
   > n1 = For Each Loop(Array:Array of User Widget Object References=GetDisplayedEntryWidgets(self=Get LockerTiles())) -> (Array Element:User Widget Object Reference)
   > n2 = Cast To WBP_UIKit_ItemCard_Locker_Jam(Object=@n1)
   > ToggleContextMenuIfSelected(self:WBP UIKit Item Card Locker Jam Object Reference=@n2)
  ]}
]
variables [
 AllowNavOnFocus {type=boolean, default=true}
 Cached WBP UIKit Item Card Locker Jam {type=WBP_UIKit_ItemCard_Locker_Jam_C, default=null}
 CachedPreviewWidget {type=FortItemPreviewWidget, default=null}
 GroupName {type=string, default=""}
 Hovered Item {type=WBP_UIKit_ItemCard_Locker_Jam_C, default=null}
 IsEmoteWheel {type=boolean}
 Item Count Per Group {type="array<integer>"}
 Last Focused Column Index {type=integer}
 List Item Object {type=Object, default=null}
]
widgets {
 info {
  inheritedWidgetCount=0
  parentClass=/Locker/MPL/ItemLists/WBP_LockerList_Base.WBP_LockerList_Base_C
  rootWidgetClass=&GridPanel
  widgetCount=4
 }
 widgets [
  GridPanel_0 {widgetClassPath=&GridPanel, widget=GridPanel_0, properties {columnFill [1], rowFill [1]}}
  LockerTiles {
   widgetClassPath=/Script/FortniteUI.FortTileView
   bIsVariable=true
   parent=GridPanel_0
   widget=LockerTiles
   properties {
    allowOverscroll=false
    bAllowDragging=false
    bConsumeTouchInputWhenScrolling=false
    bEnableRightClickScrolling=false
    bEnableScrollAnimation=false
    bEnableTouchScrolling=false
    bIsFocusable=false
    clipping=Inherit
    entryHeight=96
    entryWidgetClass=&/WBP_UIKit_ItemCard_Locker_Jam.WBP_UIKit_ItemCard_Locker_Jam_C
    entryWidth=388
    horizontalEntrySpacing=24
    numDesignerPreviewEntries=2
    tileAlignment=Fill
    visibility=SelfHitTestInvisible
    wheelScrollMultiplier=0
   }
   slot {class=GridSlot, properties {layer=1}}
  }
  ItemListHeader {
   widgetClassPath=/Locker/MPL/ItemLists/WBP_ItemListHeader
   bIsVariable=true
   parent=GridPanel_0
   widget=ItemListHeader
   properties {visibility=Collapsed}
   slot {class=GridSlot, properties {layer=1, verticalAlignment=VAlign_Top}}
  }
  InputBlocker {
   widgetClassPath=&Image
   bIsVariable=true
   parent=GridPanel_0
   widget=InputBlocker
   properties {brush {
     bIsDynamicallyLoaded=false
     drawAs=NoDrawType
     imageType=NoImage
     mirroring=NoMirror
     resourceName=None
     resourceObject=null
     tiling=NoTile
     imageSize=<32 32>
     margin=<ltrb 0 0 0 0>
     outlineSettings {
      bUseBrushTransparency=false
      roundingType=HalfHeightRadius
      width=0
      color {colorUseRule=UseColor_Specified, specifiedColor=<rgba 0 0 0 0>}
      cornerRadii=<0 0 0 0>
     }
     tintColor {colorUseRule=UseColor_Specified, specifiedColor=<rgba 1 1 1 1>}
     uVRegion {bIsValid=false, max=<0 0>, min=<0 0>}
    }}
   slot {class=GridSlot}
  }
 ]
}
