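# Sparks locker list widget, derived from WBP_LockerList_Base, showing Jam item cards in a Fort tile view.
# It forwards card context-menu, double-click, edit-styles and details events through base delegates, and binds preview actions to hovered or selected cards.
# Outside touch input, toggling the selected context scrolls that item into view before toggling its menu.
# It also adjusts compact-UI tile height and reports non-shuffle displayed items as impression telemetry.
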WidgetBlueprint /SparksCosmetics/UI/Widgets/WBP_LockerList_Sparks
graphs [
 AddImpressionTelemetry {nodes [
   AddImpressionTelemetry!() -> (then, TabFlowRiderString:String) {is=title K2Node_FunctionEntry}
   > n1 = For Each Loop(Array:Array of User Widget Object References=GetDisplayedEntryWidgets!(self:List View Base Object Reference=Get LockerTiles() -> (LockerTiles:Fort Tile View Object Reference)) -> (ReturnValue:Array of User Widget Object References)) -> (Array Element:User Widget Object Reference)
   n2 = Get SectionContentsArray() -> (SectionContentsArray:Array of Flow Analytics Impression Section Content Structures)
   Branch(
    execute=@n1.Completed
    Condition="integer > integer"(A=Length(TargetArray:Array of Flow Analytics Impression Section Content Structures=@n2))
   )
   > "Fire Event Add Impression (Non-Localized)"!(
    execute
    SectionID:String=Get TabFlowRiderString() -> (TabFlowRiderString:String)
    SectionContents:Array of Flow Analytics Impression Section Content Structures=Get SectionContentsArray()
   ) -> (then)
   n3 = Cast To WBP_UIKit_ItemCard_Locker_Jam!(execute=@n1.LoopBody, Object=@n1) -> (then, CastFailed, AsWBP UIKit Item Card Locker Jam:WBP UIKit Item Card Locker Jam Object Reference) {is=title K2Node_DynamicCast}
   > Is Valid(InputObject=@n3)
   n4 = Get FortItemVM!(self:WBP UIKit Item Card Locker Jam Object Reference=@n3) -> (FortItemVM:Fort Item VM Object Reference)
   > Branch(Condition=NOT Boolean(A=Get bIsShuffleTile!(self:Fort Item VM Object Reference=@n4) -> (bIsShuffleTile:bool)))
   n5 = GetItemTemplateId!(self:Fort Item VM Object Reference=@n4) -> (ReturnValue:String)
   > Branch(Condition=NOT Boolean(A=IsEmpty(InString=@n5)))
   > Add(
    TargetArray:Array of Flow Analytics Impression Section Content Structures=Get SectionContentsArray()
    NewItem:"Flow Analytics Impression Section Content Structure (by ref)"=Make Flow Analytics Impression Section Content!(ContentID:Text="To Text (String)"(InString=@n5), OfferID:Text, Row:int=-1, Column:int=-1, Index:int=-1) -> (FlowAnalyticsImpressionSectionContent:Flow Analytics Impression Section Content Structure) {is=title K2Node_MakeStruct}
   ) {is=title K2Node_CallArrayFunction}
  ]}
 EventGraph {nodes [
   n1 = "On Entry Generated (LockerTiles)"!() -> (OutputDelegate, then, Widget:User Widget Object Reference) {is=title K2Node_ComponentBoundEvent}
   > n2 = Cast To WBP_UIKit_ItemCard_Locker_Jam(Object=@n1.Widget)
   > SetVisibility(
    self=Get ContextMenu_Attachment!(self:WBP UIKit Item Card Locker Jam Object Reference=@n2) -> (ContextMenu_Attachment:Fort Visual Attachment Object Reference)
    InVisibility=Select(Option 0=Visible, Option 1=Collapsed, Index=Get IsEmoteWheel() -> (IsEmoteWheel:bool))
   )
   n3 = ContextMenuOpenedBind!() -> (OutputDelegate, then, ContextMenuWidgetParentEntry:Widget Object Reference) {is=title K2Node_CustomEvent}
   > Call On Tile Context Menu Open!(execute:Exec, self:WBP Locker List Base Object Reference, Widget:Widget Object Reference=@n3.ContextMenuWidgetParentEntry) -> (then:Exec) {is=title K2Node_CallDelegate}
   n4 = ContextMenuClosedBind() {is=title K2Node_CustomEvent}
   > Call On Tile Context Menu Closed!(execute:Exec, self:WBP Locker List Base Object Reference) -> (then:Exec) {is=title K2Node_CallDelegate}
   n5 = OnItemCardHoveredBind!() -> (OutputDelegate, then, Button:Common Button Base Object Reference) {is=title K2Node_CustomEvent}
   > n6 = Cast To WBP_UIKit_ItemCard_Locker_Jam(Object=@n5.Button)
   > Set Hovered Item(Hovered Item:WBP UIKit Item Card Locker Jam Object Reference=@n6) -> (Output_Get:WBP UIKit Item Card Locker Jam Object Reference)
   n7 = Get CommonInputSubsystem()
   > n8 = Branch(Condition="Not Equal (Enum)"!(A:ECommonInputType Enum=GetCurrentInputType(self=@n7), B:ECommonInputType Enum=Gamepad) -> (ReturnValue:bool))
   n9 = Event SetupBinds!() -> (OutputDelegate, then, Widget:Widget Object Reference, PreviewWidget:Fort Item Preview Widget Object Reference) {is=title K2Node_Event}
   PreviewActions!(execute=@n8.then, self:WBP Locker List Base Object Reference, PreviewWidget:Fort Item Preview Widget Object Reference=@n9.PreviewWidget) -> (then)
   n10 = OnItemDoubleClickedBind() {is=title K2Node_CustomEvent}
   > Call Close Item Selection!(execute:Exec, self:WBP Locker List Base Object Reference) -> (then:Exec) {is=title K2Node_CallDelegate}
   n11 = OnContextMenuEditStylesClickedBind() {is=title K2Node_CustomEvent}
   > Call Context Edit Styles Clicked!(execute:Exec, self:WBP Locker List Base Object Reference) -> (then:Exec) {is=title K2Node_CallDelegate}
   n12 = GoToLockerDetailsBind!() -> (OutputDelegate, then, CustomItemDetails:User Widget Class Reference) {is=title K2Node_CustomEvent}
   > Call Go to Locker Item Details!(execute:Exec, self:WBP Locker List Base Object Reference, CustomItemDetails:User Widget Class Reference=@n12.CustomItemDetails) -> (then:Exec) {is=title K2Node_CallDelegate}
   n13 = BP_OnItemScrolledIntoView_Event!() -> (OutputDelegate, then, Item:Object Reference, Widget:User Widget Object Reference) {is=title K2Node_CustomEvent}
   > Unbind Event from On Item Scrolled Into View(self:List View Object Reference=Get LockerTiles(), Delegate=@n13) {is=title K2Node_RemoveDelegate}
   > Toggle Selected Context Internal()
   Event Construct()
   > MobileSetup()
   n14 = Event SetupTelemetry!() -> (OutputDelegate, then, TabFlowRiderString:String) {is=title K2Node_Event}
   > AddImpressionTelemetry!(execute, self:WBP Locker List Base Object Reference, TabFlowRiderString:String=@n14.TabFlowRiderString) -> (then) {is=title K2Node_CallFunction}
   n15 = Event UpdateIsEmoteWheel!() -> (OutputDelegate, then, IsEmoteWheel:bool) {is=title K2Node_Event}
   > Set IsEmoteWheel(IsEmoteWheel:bool=@n15.IsEmoteWheel) -> (Output_Get:bool)
   Event ToggleSelectedContext() {is=title K2Node_Event}
   n16 = Get CommonInputSubsystem()
   > n17 = Branch(Condition="Equal (Enum)"(A=GetCurrentInputType(self=@n16), B=Touch))
   n18 = Get Selected Item!(self:List View Object Reference=Get LockerTiles()) -> (ReturnValue:Object Reference) {is=title K2Node_CallFunction}
   > Is Valid(InputObject=@n18)
   > Unbind Event from On Item Scrolled Into View(self=Get LockerTiles(), Delegate=@n13)
   > Assign On Item Scrolled Into View!(execute:Exec, self:List View Object Reference=Get LockerTiles(), Delegate:"Delegate (by ref)"=@n13) -> (then:Exec) {is=title K2Node_AssignDelegate}
   > Scroll Item Into View!(execute, self:List View Object Reference=Get LockerTiles(), Item:Object Reference=@n18) -> (then)
   Toggle Selected Context Internal(execute=@n17.else)
   Event SetContextVisibility() {is=title K2Node_Event}
   > n19 = For Each Loop(Array=GetDisplayedEntryWidgets(self=Get LockerTiles()))
   > n20 = Cast To WBP_UIKit_ItemCard_Locker_Jam(Object=@n19)
   > SetVisibility(
    self=Get ContextMenu!(self:WBP UIKit Item Card Locker Jam Object Reference=@n20) -> (ContextMenu:WBP UIKit Context Menu Anchor Button Object Reference)
    InVisibility=Hidden
   )
   n21 = Event SetupPreviewActionBinds!() -> (OutputDelegate, then, PreviewWidget:Fort Item Preview Widget Object Reference) {is=title K2Node_Event}
   > n22 = For Each Loop with Break!(Exec:Exec, Array:Array of User Widget Object References=GetDisplayedEntryWidgets(self=Get LockerTiles()), Break:Exec=@n24.then) -> (LoopBody:Exec, Array Element:User Widget Object Reference, Array Index:int, Completed:Exec)
   > n23 = Cast To WBP_UIKit_ItemCard_Locker_Jam(Object=@n22)
   > Branch(Condition=GetSelected!(self:Common Button Base Object Reference=@n23) -> (ReturnValue:bool))
   > Set Hovered Item(Hovered Item=@n23)
   > n24 = PreviewActions(PreviewWidget=@n21.PreviewWidget)
   n25 = Event BindPreviewActions!() -> (OutputDelegate, then, Preview Widget:Fort Item Preview Widget Object Reference) {is=title K2Node_Event}
   > PreviewActions(PreviewWidget=@n25.Preview Widget)
   n26 = Cast To WBP_UIKit_ItemCard_Locker_Jam(execute=@n9.then, Object=@n9.Widget)
   Bind Event to On Context Menu Opened(execute=@n27.then_0, self:WBP UIKit Item Card Locker Jam Object Reference=@n26, Delegate=@n3) {is=title K2Node_AddDelegate}
   Bind Event to On Context Menu Closed(execute=@n27.then_1, self:WBP UIKit Item Card Locker Jam Object Reference=@n26, Delegate=@n4) {is=title K2Node_AddDelegate}
   Bind Event to On Hovered(execute=@n27.then_2, self:Common Button Base Object Reference=@n26, Delegate=@n5) {is=title K2Node_AddDelegate}
   Bind Event to On Item Double Clicked(execute=@n27.then_3, self:WBP UIKit Item Card Locker Jam Object Reference=@n26, Delegate=@n10) {is=title K2Node_AddDelegate}
   Bind Event to On Context Menu Edit Styles Clicked(execute=@n27.then_4, self:WBP UIKit Item Card Locker Jam Object Reference=@n26, Delegate=@n11) {is=title K2Node_AddDelegate}
   Bind Event to Go to Locker Item Details(execute=@n27.then_5, self:WBP UIKit Item Card Locker Jam Object Reference=@n26, Delegate=@n12) {is=title K2Node_AddDelegate}
   n27 = Sequence!(execute=@n26.then) -> (then_0, then_1, then_2:Exec, then_3:Exec, then_4:Exec, then_5:Exec, then_6:Exec)
  ]}
 GetListView {nodes [
   GetListView() {is=title K2Node_FunctionEntry}
   > Return Node!(execute, TileView:Fort Tile View Object Reference=Get LockerTiles())
  ]}
 MobileSetup {nodes [
   MobileSetup() {is=title K2Node_FunctionEntry}
   > Branch(Condition=IsUsingCompactUI())
   > SetEntryHeight!(execute, self:Tile View Object Reference=Get LockerTiles(), NewHeight:float=150) -> (then)
  ]}
 PreviewActions {nodes [
   n1 = PreviewActions!() -> (then, PreviewWidget:Fort Item Preview Widget Object Reference) {is=title K2Node_FunctionEntry}
   n2 = Get Hovered Item() -> (Hovered Item:WBP UIKit Item Card Locker Jam Object Reference)
   > Is Valid(InputObject=@n2)
   > Bind Preview Actions!(
    execute
    self:WBP UIKit Item Card Base Object Reference=@n2
    Preview Widget:Fort Item Preview Widget Object Reference=@n1
    Context Menu:WBP UIKit Context Menu Anchor Button Base Object Reference=Get ContextMenu(self=@n2)
    Item VM:Fort Item VM Object Reference=Get FortItemVM(self=@n2)
    Include Special Actions:bool=true
    Include Toggle Actions:bool=true
   ) -> (then)
  ]}
 Toggle Selected Context Internal {nodes [
   Toggle Selected Context Internal() {is=title K2Node_FunctionEntry}
   > n1 = For Each Loop(Array=GetDisplayedEntryWidgets(self=Get LockerTiles()))
   > n2 = Cast To WBP_UIKit_ItemCard_Locker_Jam(Object=@n1)
   > ToggleContextMenuIfSelected(self:WBP UIKit Item Card Locker Jam Object Reference=@n2)
  ]}
]
variables [
 Hovered Item {type=WBP_UIKit_ItemCard_Locker_Jam_C, default=null}
 IsEmoteWheel {type=boolean}
]
widgets {
 info {
  inheritedWidgetCount=0
  parentClass=/Locker/MPL/ItemLists/WBP_LockerList_Base.WBP_LockerList_Base_C
  rootWidgetClass=/Script/FortniteUI.FortTileView
  widgetCount=1
 }
 widgets [LockerTiles {
   widgetClassPath=&FortTileView
   bIsVariable=true
   widget=LockerTiles
   properties {
    clipping=Inherit
    entryHeight=112
    entryWidgetClass=&/WBP_UIKit_ItemCard_Locker_Jam.WBP_UIKit_ItemCard_Locker_Jam_C
    entryWidth=388
    horizontalEntrySpacing=23
    navigation=&self:WidgetTree.LockerTiles.WidgetNavigation_0
    numDesignerPreviewEntries=20
    scrollbarDisabledVisibility=Hidden
    tileAlignment=Fill
    verticalEntrySpacing=16
   }
  }]
}
