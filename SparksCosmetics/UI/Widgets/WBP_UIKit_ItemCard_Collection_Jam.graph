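# Wide item card widget for Jam cosmetic collection lists, extending the UIKit slim item card base and bound to a FortItemVM list item.
# InitTileLayout re-grids inherited slots into a tile with title, description and misc text.
# Hover, selection and input-method changes drive text scrolling, a background highlight and a slide/fade reveal of misc text, with details shown by default on touch.
# It also shows secondary equipped slot indexes and fires interaction analytics.
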WidgetBlueprint /SparksCosmetics/UI/Widgets/WBP_UIKit_ItemCard_Collection_Jam
animations [
 Anim_Highlight {
  binding_count=2
  length_seconds=0.2004
  object=&self:Anim_Highlight
  sequence {bindings [Background_Backer {
     id=#1
     possessed_class=Image
     tracks [{
       type=MovieSceneWidgetMaterialTrack
       sections [{
         type=MovieSceneParameterSection
         end_frame=4
         start_frame=0
         channels [
          Additive Percentage {type=MovieSceneScriptingFloatChannel, const [0.4,1]}
          BgColor_1R {type=MovieSceneScriptingFloatChannel, keys [[0,0.3646],[4,1]]}
          BgColor_1G {type=MovieSceneScriptingFloatChannel, keys [[0,0.3646],[4,1]]}
          BgColor_1B {type=MovieSceneScriptingFloatChannel, keys [[0,0.3646],[4,1]]}
          BgColor_1A {type=MovieSceneScriptingFloatChannel, const [1,2]}
          BgColor_2R {type=MovieSceneScriptingFloatChannel, keys [[0,0.3663],[4,1]]}
          BgColor_2G {type=MovieSceneScriptingFloatChannel, keys [[0,0.3663],[4,1]]}
          BgColor_2B {type=MovieSceneScriptingFloatChannel, keys [[0,0.3663],[4,1]]}
          BgColor_2A {type=MovieSceneScriptingFloatChannel, const [1,2]}
         ]
        }]
      }]
    }]}
 }
 Anim_DisplayDetails {
  binding_count=2
  length_seconds=0.2
  object=&self:Anim_DisplayDetails
  sequence {bindings [
    VerticalBox_0 {
     id=#2
     possessed_class=VerticalBox
     tracks [{
       type=MovieScene2DTransformTrack
       sections [{
         type=MovieScene2DTransformSection
         end_frame=4
         start_frame=0
         channels [
          Translation.X {type=MovieSceneScriptingFloatChannel, const [0,1]}
          Translation.Y {type=MovieSceneScriptingFloatChannel, keys [[0,13],[4,0]]}
          Angle {type=MovieSceneScriptingFloatChannel, const [0,1]}
          Scale.X {type=MovieSceneScriptingFloatChannel, const [1,1]}
          Scale.Y {type=MovieSceneScriptingFloatChannel, const [1,1]}
          Shear.X {type=MovieSceneScriptingFloatChannel, const [0,1]}
          Shear.Y {type=MovieSceneScriptingFloatChannel, const [0,1]}
         ]
        }]
      }]
    }
    Text_Misc {
     id=#3
     possessed_class=CommonTextBlock
     tracks [{
       type=MovieSceneFloatTrack
       display_name=Render Opacity
       sections [{type=MovieSceneFloatSection, end_frame=4, start_frame=0, channels [None {type=MovieSceneScriptingFloatChannel, keys [[0,0],[4,1]]}]}]
      }]
    }
   ]}
 }
]
graphs [
 AnalyticsItemInteracted {nodes [
   n1 = AnalyticsItemInteracted!() -> (then, Context:String) {is=title K2Node_FunctionEntry}
   n2 = GetItemTemplateId!(self:Fort Item VM Object Reference=Get FortItemVM() -> (FortItemVM:Fort Item VM Object Reference)) -> (ReturnValue:String)
   > FireEventElementInteractedByName!(
    execute
    Element Name:String="To String (Text)"(InText=Format Text!(Format="LOCTEXT(\"{0} {1}\")", 0:String=@n1, 1:String=@n2) -> (Result))
    WidgetCallingFrom:Widget Object Reference=Self-Reference()
   ) -> (then, Success:bool)
  ]}
 CheckBackgroundType {nodes [
   n1 = CheckBackgroundType!() -> (then, IsNewPassItem:bool=~, Material:Material Interface Soft Object Reference) {is=title K2Node_FunctionEntry}
   > n2 = Branch(Condition=@n1)
   > Return Node!(execute, Background:Material Interface Soft Object Reference=/FortUILibrary/MaterialInstances/Effects/MI_UI_RewardState)
   Return Node(execute=@n2.else, Background=Get Material() -> (Material:Material Interface Soft Object Reference))
  ]}
 EventGraph {nodes [
   n1 = Event PreConstruct()
   > Parent: PreConstruct(IsDesignTime=@n1.IsDesignTime)
   > Set RenderOpacity!(
    execute
    RenderOpacity:float=SelectFloat(A=1, B=0, bPickA=@n1.IsDesignTime)
    self:Widget Object Reference=Get Text_Misc() -> (Text_Misc:Common Text Object Reference)
   ) -> (then, Output_Get:float)
   > InitTileLayout()
   > SetDetailsState!(execute, self, ShowDetails:bool=IsUsingTouch(PlayerController=GetOwningPlayer())) -> (then)
   > Branch(Condition=IsMobileGame())
   > MobileSetup()
   n2 = Event OnListItemObjectSet()
   > n3 = Cast To FortItemVM!(execute, Object=@n2.ListItemObject) -> (then, CastFailed, AsFort Item VM:Fort Item VM Object Reference) {is=title K2Node_DynamicCast}
   > InitializeItemCard(self:WBP UIKit Slim Item Card Base Object Reference)
   > SetFortItemVM!(execute, self, Viewmodel:Fort Item VM Object Reference=@n3) -> (then)
   Event Construct()
   > Parent: Construct()
   > SetTextScroll!(execute, self, TitleScroll:bool=false, DescriptionScroll:bool=false) -> (then)
   Event Destruct()
   > Parent: Destruct()
   n4 = Event OnRemovedFromFocusPath!() -> (OutputDelegate, then, InFocusEvent:Focus Event Structure) {is=title K2Node_Event}
   > Parent: OnRemovedFromFocusPath!(execute, InFocusEvent:Focus Event Structure=@n4.InFocusEvent) -> (then) {is=title K2Node_CallParentFunction}
   > SetTextScroll(TitleScroll=false, DescriptionScroll=false)
   Event On Deselected()
   > Parent: On Deselected() {is=title K2Node_CallParentFunction}
   > Branch(Condition=IsUsingTouch(PlayerController=GetOwningPlayer()))
   > PlaySelectionAnimation!(execute, self, IsSelected:bool=false, Instant Animation:bool=false) -> (then)
   Event On Clicked()
   > Parent: On Clicked() {is=title K2Node_CallParentFunction}
   Event On Double Clicked() {is=title K2Node_Event}
   > Parent: On Double Clicked() {is=title K2Node_CallParentFunction}
   > Call On Item Double Clicked!(execute:Exec, self:WBP UIKit Item Card Collection Jam Object Reference) -> (then:Exec) {is=title K2Node_CallDelegate}
   Event On Hovered()
   > n5 = Branch(Condition=IsUsingTouch(PlayerController=GetOwningPlayer()))
   Set IsItemCardLockerHovered(execute=@n5.else, IsItemCardLockerHovered:bool=true) -> (Output_Get:bool)
   > SetTextScroll(TitleScroll=true, DescriptionScroll=true)
   > Parent: On Hovered() {is=title K2Node_CallParentFunction}
   > PlaySelectionAnimation(IsSelected=true, Instant Animation=false)
   Event On Unhovered()
   > n6 = Branch(Condition=IsUsingTouch(PlayerController=GetOwningPlayer()))
   Set IsItemCardLockerHovered(execute=@n6.else, IsItemCardLockerHovered=false)
   > SetTextScroll(TitleScroll=false, DescriptionScroll=false)
   > Parent: On Unhovered() {is=title K2Node_CallParentFunction}
   > PlaySelectionAnimation(IsSelected=false, Instant Animation=false)
   n7 = Event On Input Method Changed!() -> (OutputDelegate, then, CurrentInputType:ECommonInputType Enum) {is=title K2Node_Event}
   > Is Valid(InputObject=Get FortItemVM())
   > n8 = Switch on ECommonInputType!(execute, Selection:ECommonInputType Enum=@n7.CurrentInputType) -> (MouseAndKeyboard:Exec, Gamepad:Exec, Touch:Exec, Count:Exec) {is=title K2Node_SwitchEnum}
   SetDetailsState(execute=@n8.Gamepad, ShowDetails=GetSelected!(self:Common Button Base Object Reference) -> (ReturnValue:bool))
   n9 = Branch(execute=@n8.MouseAndKeyboard, Condition=GetSelected())
   > n10 = SetFocusState!(execute, self:WBP UIKit Slim Item Card Base Object Reference, IsFocused:bool=false) -> (then)
   SetDetailsState(execute=[@n10.then, @n9.else], ShowDetails=false)
   SetDetailsState(execute=@n8.Touch, ShowDetails=true)
   > Branch(Condition=GetSelected())
   > SetFocusState(IsFocused=true)
   n11 = Event OnAddedToFocusPath!() -> (OutputDelegate, then, InFocusEvent:Focus Event Structure) {is=title K2Node_Event}
   > Parent: OnAddedToFocusPath!(execute, InFocusEvent:Focus Event Structure=@n11.InFocusEvent) -> (then) {is=title K2Node_CallParentFunction}
   > SetTextScroll(TitleScroll=true, DescriptionScroll=true)
   > n12 = Switch on ECommonInputType(Selection=GetCurrentInputType(self=Get CommonInputSubsystem()))
   SetFocusState(execute=@n12.Touch, IsFocused=GetSelected())
  ]}
 GoToLockerItemDetails {nodes [GoToLockerItemDetails!() -> (then, CustomItemDetails:User Widget Class Reference) {is=title K2Node_FunctionEntry}]}
 InitTileLayout {nodes [
   InitTileLayout() {is=title K2Node_FunctionEntry}
   n1 = Get Root() -> (Root:Grid Panel Object Reference)
   > SetColumnFill!(execute, self:Grid Panel Object Reference=@n1, ColumnIndex:int=0, Coefficient:float=0) -> (then)
   > SetColumnFill(self=@n1, ColumnIndex=1, Coefficient=1)
   > SetGridSlotColumnAndRow!(
    execute
    Widget:Widget Object Reference=Get Block_Outline() -> (Block_Outline:WBP UIKit Block Outline Object Reference)
    Row:int=0
    Row Span:int=0
    Column:int=0
    Column Span:int=2
   ) -> (then)
   > SetGridSlotColumnAndRow(Widget=Get Slot_EquippedOutline() -> (Slot_EquippedOutline:Named Slot Object Reference), Row=0, Row Span=0, Column=0, Column Span=2)
   > SetGridSlotColumnAndRow(Widget=Get Slot_Background() -> (Slot_Background:Named Slot Object Reference), Row=0, Row Span=0, Column=0, Column Span=2)
   > SetGridSlotColumnAndRow(Widget=Get Slot_Custom() -> (Slot_Custom:Named Slot Object Reference), Row=0, Row Span=0, Column=1, Column Span=1)
   > n2 = AddChildToHorizontalBox!(
    execute
    self:Horizontal Box Object Reference=Get TopStatusIndicators() -> (TopStatusIndicators:Horizontal Box Object Reference)
    Content:Widget Object Reference=Get RightIndicators() -> (RightIndicators:Grid Panel Object Reference)
   ) -> (then, ReturnValue:Horizontal Box Slot Object Reference)
   > SetPadding!(
    execute
    self:Horizontal Box Slot Object Reference=@n2
    InPadding_Left:float=8
    InPadding_Top:float=0
    InPadding_Right:float=0
    InPadding_Bottom:float=0
   ) -> (then)
  ]}
 Is Item Equipped {nodes [
   n1 = Is Item Equipped!() -> (then, Category Item:Fort Item VM Object Reference, Item VM:Fort Item VM Object Reference) {is=title K2Node_FunctionEntry}
   > Return Node!(execute, NewParam:bool="Equal (Object)"(A=@n1, B=@n1.Item VM))
  ]}
 MobileSetup {nodes [
   MobileSetup() {is=title K2Node_FunctionEntry}
   > SetSize!(
    execute
    self:Spacer Object Reference=Get Image_Stretcher() -> (Image_Stretcher:Spacer Object Reference)
    InSize:Vector 2D Structure=(X=280,Y=0)
   ) -> (then)
  ]}
 OnContextMenuClosed {nodes [OnContextMenuClosed() {is=title K2Node_FunctionEntry}]}
 OnContextMenuEditStylesClicked {nodes [OnContextMenuEditStylesClicked() {is=title K2Node_FunctionEntry}]}
 OnContextMenuOpened {nodes [OnContextMenuOpened!() -> (then, ContextMenuWidgetParentEntry:Widget Object Reference) {is=title K2Node_FunctionEntry}]}
 OnItemCardHovered {nodes [OnItemCardHovered() {is=title K2Node_FunctionEntry}]}
 OnItemDoubleClicked {nodes [OnItemDoubleClicked() {is=title K2Node_FunctionEntry}]}
 PlaySelectionAnimation {nodes [
   n1 = PlaySelectionAnimation!() -> (then, IsSelected:bool=~, Instant Animation:bool=~) {is=title K2Node_FunctionEntry}
   n2 = Get Anim_Highlight() -> (Anim_Highlight:Widget Animation Object Reference)
   n3 = Select(Option 0:EUMGSequencePlayMode Enum=Reverse, Option 1:EUMGSequencePlayMode Enum, Index=@n1) -> (ReturnValue:EUMGSequencePlayMode Enum)
   > PlayAnimation(InAnimation=@n2, StartAtTime=SelectFloat(A=GetEndTime(self=@n2), B=0, bPickA=@n1.Instant Animation), PlayMode=@n3)
   > n4 = Branch(Condition=IsUsingTouch(PlayerController=GetOwningPlayer()))
   n5 = Get Anim_DisplayDetails() -> (Anim_DisplayDetails:Widget Animation Object Reference)
   > PlayAnimation(InAnimation=@n5, StartAtTime=GetEndTime(self=@n5), PlayMode=Select(Option 1=Reverse, Index=Get IsMiscTextEmpty() -> (IsMiscTextEmpty:bool)))
   n6 = Get Anim_DisplayDetails()
   n7 = Get IsMiscTextEmpty()
   PlayAnimation(
    execute=@n4.else
    InAnimation=@n6
    StartAtTime=SelectFloat(A=GetEndTime(self=@n6), B=0, bPickA=@n7)
    PlayMode=Select(Option 0=@n3, Option 1=Reverse, Index=@n7)
   )
  ]}
 ResetAnimationState {nodes [
   ResetAnimationState() {is=title K2Node_FunctionEntry}
   > PlayAnimationDuringTime!(
    execute
    Widget:User Widget Object Reference=Self-Reference()
    Animation:Widget Animation Object Reference=Get Anim_Hovered_Base() -> (Anim_Hovered_Base:Widget Animation Object Reference)
    PlayTime:float=0
    bPlayForward:bool=false
   ) -> (then)
   > ResetAnimationState(self:WBP UIKit Block Outline Object Reference=Get Block_Outline()) {is=title K2Node_CallFunction}
  ]}
 SelectCard {nodes [
   SelectCard() {is=title K2Node_FunctionEntry}
   > PlaySelectionAnimation(IsSelected=true, Instant Animation=false) {is=title K2Node_CallFunction}
   > SetFocusState(IsFocused=true)
   > n1 = Cast To ListView!(execute, Object=GetOwningListView!(UserListEntry:User List Entry Interface) -> (ReturnValue:List View Base Object Reference)) -> (then, CastFailed, AsList View:List View Object Reference) {is=title K2Node_DynamicCast}
   > Set Selected Item!(
    execute
    self:List View Object Reference=@n1
    Item:Object Reference=GetListItemObject!(UserObjectListEntry:User Object List Entry Interface) -> (ReturnValue:Object Reference)
   ) -> (then) {is=title K2Node_CallFunction}
  ]}
 Set Text Description {nodes [
   n1 = Set Text Description!() -> (then, InText:Text) {is=title K2Node_FunctionEntry}
   n2 = Get Text_Description() -> (Text_Description:Common Text Object Reference)
   > SetVisibility(
    self=@n2
    InVisibility=Select(Option 0:ESlateVisibility Enum=HitTestInvisible, Option 1:ESlateVisibility Enum=Collapsed, Index=TextIsEmpty(InText=@n1)) -> (ReturnValue:ESlateVisibility Enum)
   )
   > "SetText (Text)"(self=@n2, InText=@n1)
  ]}
 SetDetailsState {nodes [
   n1 = SetDetailsState!() -> (then, ShowDetails:bool=~) {is=title K2Node_FunctionEntry}
   n2 = Get Anim_DisplayDetails()
   > PlayAnimation(
    InAnimation=@n2
    StartAtTime=GetEndTime(self=@n2)
    PlayMode=Select(Option 0:EUMGSequencePlayMode Enum=Reverse, Option 1:EUMGSequencePlayMode Enum, Index=@n1) -> (ReturnValue:EUMGSequencePlayMode Enum)
   )
  ]}
 SetImage {nodes [
   n1 = SetImage!() -> (then, Item Image:Texture 2D Soft Object Reference) {is=title K2Node_FunctionEntry}
   > SetItemImage!(execute, self:WBP UIKit Slim Item Card Base Object Reference, Item Image:Soft Object Reference=@n1) -> (then)
   > n2 = GetDynamicMaterial(self=Get LazyImage_Item() -> (LazyImage_Item:Fort Lazy Image Object Reference))
   > SetScalarParameterValue(self=@n2, ParameterName=TopRightCornerRadius, Value=0)
   > SetScalarParameterValue(self=@n2, ParameterName=BottomRightCornerRadius, Value=0)
  ]}
 SetMiscText {nodes [
   n1 = SetMiscText!() -> (then, Text:Text) {is=title K2Node_FunctionEntry}
   n2 = TextIsEmpty(InText=@n1)
   > Set IsMiscTextEmpty(IsMiscTextEmpty:bool=@n2) -> (Output_Get:bool)
   n3 = Get Text_Misc()
   > SetVisibility(
    self=@n3
    InVisibility=Select(Option 0:ESlateVisibility Enum=HitTestInvisible, Option 1:ESlateVisibility Enum=Collapsed, Index=@n2) -> (ReturnValue:ESlateVisibility Enum)
   )
   > "SetText (Text)"(self=@n3, InText=@n1)
   > PlaySelectionAnimation(IsSelected=GetSelected(), Instant Animation=true)
  ]}
 SetTextScroll {nodes [
   n1 = SetTextScroll!() -> (then, TitleScroll:bool=true, DescriptionScroll:bool=true) {is=title K2Node_FunctionEntry}
   > SetScrollingEnabled!(execute, self:Common Text Object Reference=Get Text_Title() -> (Text_Title:Common Text Object Reference), bInIsScrollingEnabled:bool=@n1) -> (then)
   > SetScrollingEnabled(self=Get Text_Description(), bInIsScrollingEnabled=@n1.DescriptionScroll)
  ]}
 Update Equipped Indexes {nodes [
   n1 = Update Equipped Indexes!() -> (then, Equipped Indexes:Array of Integers) {is=title K2Node_FunctionEntry}
   > n2 = Branch(Condition=Is Empty!(TargetArray:Array of Integers=@n1) -> (ReturnValue:bool))
   > SetItemIsEquippedSecondary!(
    execute
    self:WBP UIKit Slim Item Card Base Object Reference
    Is Equipped:bool=false
    ShowEquippedSlotIndex:bool=false
    EquippedSlotIndex:int=0
    SlotIcon:Soft Object Reference
   ) -> (then)
   n3 = Get Equipped Indexes() -> (Equipped Indexes:Array of Integers)
   n4 = Get FortItemVM()
   SetItemIsEquippedSecondary(
    execute=@n2.else
    Is Equipped=true
    ShowEquippedSlotIndex=Get bSecondaryEquipShowNumbers!(self:Fort Locker Category Config Object Reference=GetItemCategory!(self:Fort Item VM Object Reference=@n4) -> (ReturnValue:Fort Locker Category Config Object Reference)) -> (bSecondaryEquipShowNumbers:bool)
    EquippedSlotIndex=int + int(A="Get (a copy)"(Array:Array of Integers=@n3) -> (Output:int))
   )
  ]}
]
variables [
 ClearDynamicMaterial {type="array<MaterialInstanceDynamic>", default []}
 GoToLockerItemDetails {}
 IsItemCardLockerHovered {type=boolean}
 IsMiscTextEmpty {type=boolean, default=true}
 Item Material {type=MaterialInstance, default=null}
 OnContextMenuClosed {}
 OnContextMenuEditStylesClicked {}
 OnContextMenuOpened {}
 OnItemCardHovered {}
 OnItemDoubleClicked {}
 OtherCategoryItems {type="map<FortItemVM>", default {}}
]
widgets {
 info {
  inheritedWidgetCount=0
  namedSlotCount=3
  parentClass=/FortUILibrary/Systems/Items/WBP_UIKit_SlimItemCard_Base.WBP_UIKit_SlimItemCard_Base_C
  rootWidgetClass=null
  widgetCount=11
 }
 widgets [
  ContextMenu_Attachment {
   widgetClassPath=&FortVisualAttachment
   bIsVariable=true
   widget=ContextMenu_Attachment
   properties {bOverride_HeightOverride=true, bOverride_WidthOverride=true, heightOverride=40, visibility=Visible, widthOverride=40, contentAnchor=<1 1>}
   slot=~
  }
  ContextMenu {
   widgetClassPath=/FortUILibrary/UIKit/ContextMenu/WBP_UIKit_ContextMenuAnchorButton
   bIsVariable=true
   parent=ContextMenu_Attachment
   widget=ContextMenu
   properties {
    button Background Material=/FortUILibrary/UIKit/Buttons/MaterialInstances/CTAButtons/MI_UIKit_BgQuiet
    buttonClass=/FortUILibrary/UIKit/Buttons/WBP_UIKit_Button_Contextual.WBP_UIKit_Button_Contextual_C
    buttonIcon=/FortUILibrary/Textures/Icons/T_UI_Icon_UIKit_ContextMenu
    buttonText=""
    button Padding=<ltrb 4 4 4 4>
    button Text Font {
     bForceMonospaced=false
     bMaterialIsStencil=false
     fontMaterial=/FortUILibrary/UIKit/Buttons/MaterialInstances/GenericButtons/MI_UIKit_GenericText_Quiet
     fontObject=&Burbank
     letterSpacing=180
     monospacedWidth=1
     size=26.25
     skewAmount=0
     typefaceFontName=BurbankRegular-Black
     outlineSettings {
      bApplyOutlineToDropShadows=false
      bMiteredCorners=false
      bSeparateFillAlpha=false
      outlineMaterial=null
      outlineSize=0
      outlineColor=<rgba 0 0 0 1>
     }
    }
   }
   slot {class=SizeBoxSlot, properties {horizontalAlignment=HAlign_Right, verticalAlignment=VAlign_Bottom, padding=<ltrb -8 -8 8 8>}}
  }
  Overlay_50 {widgetClassPath=&Overlay, widget=Overlay_50}
  Background_Backer {
   widgetClassPath=&Image
   bIsVariable=true
   parent=Overlay_50
   widget=Background_Backer
   properties {brush {
     bIsDynamicallyLoaded=false
     drawAs=Image
     imageType=NoImage
     mirroring=NoMirror
     resourceName=None
     resourceObject=/FortUILibrary/Systems/Items/MaterialInstances/MI_UIKit_Item_Background
     tiling=NoTile
     imageSize=<32 32>
     margin=<ltrb 0 0 0 0>
     outlineSettings {
      bUseBrushTransparency=false
      roundingType=HalfHeightRadius
      width=0
      color {colorUseRule=UseColor_Specified, specifiedColor=<rgba 0 0 0 0>}
      cornerRadii=<0 0 0 0>
     }
     tintColor {colorUseRule=UseColor_Specified, specifiedColor=<rgba 1 1 1 1>}
     uVRegion {bIsValid=false, max=<0 0>, min=<0 0>}
    }}
   slot {class=OverlaySlot, properties {horizontalAlignment=HAlign_Fill, verticalAlignment=VAlign_Fill}}
  }
  Image_Stretcher {
   widgetClassPath=&Spacer
   bIsVariable=true
   parent=Overlay_50
   widget=Image_Stretcher
   properties {size=<192 1>}
   slot {class=OverlaySlot, properties {horizontalAlignment=HAlign_Fill, verticalAlignment=VAlign_Fill}}
  }
  GridPanel_79 {widgetClassPath=&GridPanel, widget=GridPanel_79, properties {columnFill [1], rowFill [0,1]}}
  TopStatusIndicators {
   widgetClassPath=&HorizontalBox
   bIsVariable=true
   parent=GridPanel_79
   widget=TopStatusIndicators
   slot {class=GridSlot, properties {horizontalAlignment=HAlign_Right, layer=2, row=1, verticalAlignment=VAlign_Top, padding=<ltrb 0 8 8 0>}}
  }
  VerticalBox_0 {
   widgetClassPath=&VerticalBox
   parent=GridPanel_79
   widget=VerticalBox_0
   properties {clipping=ClipToBounds}
   slot {class=GridSlot, properties {row=1, verticalAlignment=VAlign_Center, padding=<ltrb 12 -8 30 -6>}}
  }
  Text_Title {
   widgetClassPath=&CommonTextBlock
   parent=VerticalBox_0
   widget=Text_Title
   properties {
    mobileFontSizeMultiplier=1.2
    scrollStyle=&TextScrollStyle-Base_C
    text="[PH] Item Name"
    textOverflowPolicy=Ellipsis
    font {
     bForceMonospaced=false
     bMaterialIsStencil=false
     fontMaterial=/FortUILibrary/UIKit/Buttons/MaterialInstances/CTAButtons/MI_UIKit_TextRegular
     fontObject=&HeadingNow
     letterSpacing=0
     monospacedWidth=1
     size=20.25
     skewAmount=0
     typefaceFontName=75 Medium
     outlineSettings {
      bApplyOutlineToDropShadows=false
      bMiteredCorners=false
      bSeparateFillAlpha=false
      outlineMaterial=null
      outlineSize=0
      outlineColor=<rgba 0 0 0 1>
     }
    }
    shadowOffset=<0 0>
    strikeBrush {
     bIsDynamicallyLoaded=false
     drawAs=NoDrawType
     imageType=NoImage
     mirroring=NoMirror
     resourceName=None
     resourceObject=null
     tiling=NoTile
     imageSize=<32 32>
     margin=<ltrb 0 0 0 0>
     outlineSettings {
      bUseBrushTransparency=false
      roundingType=HalfHeightRadius
      width=0
      color {colorUseRule=UseColor_Specified, specifiedColor=<rgba 0 0 0 0>}
      cornerRadii=<0 0 0 0>
     }
     tintColor {colorUseRule=UseColor_Specified, specifiedColor=<rgba 1 1 1 1>}
     uVRegion {bIsValid=false, max=<0 0>, min=<0 0>}
    }
   }
   slot {class=VerticalBoxSlot, properties {verticalAlignment=VAlign_Bottom}}
  }
  Text_Description {
   widgetClassPath=&CommonTextBlock
   bIsVariable=true
   parent=VerticalBox_0
   widget=Text_Description
   properties {
    mobileFontSizeMultiplier=1.3
    scrollStyle=&TextScrollStyle-Base_C
    text="[PH] Item Description"
    textOverflowPolicy=Ellipsis
    font {
     bForceMonospaced=false
     bMaterialIsStencil=false
     fontMaterial=&MI_UIKit_TextRegular
     fontObject=&HeadingNow
     letterSpacing=0
     monospacedWidth=1
     size=15.75
     skewAmount=0
     typefaceFontName=74 Regular
     outlineSettings {
      bApplyOutlineToDropShadows=false
      bMiteredCorners=false
      bSeparateFillAlpha=false
      outlineMaterial=null
      outlineSize=0
      outlineColor=<rgba 0 0 0 1>
     }
    }
    shadowOffset=<0 0>
    strikeBrush {
     bIsDynamicallyLoaded=false
     drawAs=NoDrawType
     imageType=NoImage
     mirroring=NoMirror
     resourceName=None
     resourceObject=null
     tiling=NoTile
     imageSize=<32 32>
     margin=<ltrb 0 0 0 0>
     outlineSettings {
      bUseBrushTransparency=false
      roundingType=HalfHeightRadius
      width=0
      color {colorUseRule=UseColor_Specified, specifiedColor=<rgba 0 0 0 0>}
      cornerRadii=<0 0 0 0>
     }
     tintColor {colorUseRule=UseColor_Specified, specifiedColor=<rgba 1 1 1 1>}
     uVRegion {bIsValid=false, max=<0 0>, min=<0 0>}
    }
   }
   slot {class=VerticalBoxSlot, properties {verticalAlignment=VAlign_Center}}
  }
  Text_Misc {
   widgetClassPath=&CommonTextBlock
   bIsVariable=true
   parent=VerticalBox_0
   widget=Text_Misc
   properties {
    bAutoCollapseWithEmptyText=true
    mobileFontSizeMultiplier=1.3
    renderOpacity=0
    text="[PH] Misc Text"
    textOverflowPolicy=Ellipsis
    colorAndOpacity {colorUseRule=UseColor_Specified, specifiedColor=<rgba 1 1 1 0.6>}
    font {
     bForceMonospaced=false
     bMaterialIsStencil=false
     fontMaterial=&MI_UIKit_TextRegular
     fontObject=&HeadingNow
     letterSpacing=0
     monospacedWidth=1
     size=14.25
     skewAmount=0
     typefaceFontName=74 Regular
     outlineSettings {
      bApplyOutlineToDropShadows=false
      bMiteredCorners=false
      bSeparateFillAlpha=false
      outlineMaterial=null
      outlineSize=0
      outlineColor=<rgba 0 0 0 1>
     }
    }
    shadowOffset=<0 0>
    strikeBrush {
     bIsDynamicallyLoaded=false
     drawAs=NoDrawType
     imageType=NoImage
     mirroring=NoMirror
     resourceName=None
     resourceObject=null
     tiling=NoTile
     imageSize=<32 32>
     margin=<ltrb 0 0 0 0>
     outlineSettings {
      bUseBrushTransparency=false
      roundingType=HalfHeightRadius
      width=0
      color {colorUseRule=UseColor_Specified, specifiedColor=<rgba 0 0 0 0>}
      cornerRadii=<0 0 0 0>
     }
     tintColor {colorUseRule=UseColor_Specified, specifiedColor=<rgba 1 1 1 1>}
     uVRegion {bIsValid=false, max=<0 0>, min=<0 0>}
    }
   }
   slot {class=VerticalBoxSlot, properties {verticalAlignment=VAlign_Top}}
  }
 ]
}
